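# Unlit, two-sided translucent material for a graffiti fountain effect in the Sparks stage show, usable on Niagara sprites and mesh particles.
# A texture panning along V is tinted by particle color.
# A saturated V-coordinate gradient softens one end, and particle alpha, texture alpha and a dynamic Alpha parameter combine with it for opacity.
# A time-of-day brightness multiplier on the tinted color feeds the shared FX emissive adjustment function.
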Material /SparksMaps/Effects/StageShow/Graffiti/M_Sparks_StageShow_Graffiti_Fountain
graph {
 properties {
  bAutomaticallySetUsageInEditor=false
  bCastRayTracedShadows=false
  bUsedWithNiagaraMeshParticles=true
  bUsedWithNiagaraSprites=true
  blendMode=BLEND_Translucent
  shadingModel=MSM_Unlit
  twoSided=true
  usedShadingModels=Unlit
 }
 nodes [
  n1 = ScalarParameter() {defaultValue=1, parameterName=Morn}
  n2 = ScalarParameter() {defaultValue=1, parameterName=Day}
  n3 = ScalarParameter() {defaultValue=1, parameterName=Evening}
  n4 = ScalarParameter() {defaultValue=1, parameterName=Night}
  n5 = ParticleColor()
  n6 = TextureSampleParameter2D(UVs=Panner(Coordinate=TextureCoordinate() {vTiling=0.25}) {speedY=1}) {parameterName=Texture, texture=&/T_Sparks_StageShow_Graffiti_Fountain}
  n7 = Multiply(A=@n5.RGB, B=@n6.RGB)
  MaterialFunctionCall(Emissive=Multiply(
    A=MaterialFunctionCall(Color_Morning=@n1, Color_Day=@n2, Color_Evening=@n3, Color_Night=@n4) {materialFunction=&Color_Time_of_Day}.Result
    B=@n7
   )) {materialFunction=&MF_FX_EmissiveAdjustments}
  n8 = ComponentMask(None=TextureCoordinate()) {g=true}
  n9 = ScalarParameter() {defaultValue=0.1, parameterName=Param_2}
  n10 = Saturate(None=Divide(A=Subtract(A=@n8, B=@n9), B=ScalarParameter() {defaultValue=0.2, parameterName=Param_3}))
  Multiply(A=Multiply(A=Multiply(A=@n5.A, B=@n6.A), B=@n10), B=DynamicParameter() {paramNames [Alpha,Param2,Param3,Param4]}.Alpha)
  Multiply(A=@n7, B=@n10)
 ]
}
